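# Translucent UI material for a Season 37 battle pass quest level tag, drawing a glowing logo shape with complete and locked states.
# Two custom rhombus SDFs, one rotated, smooth-union into a shape that yields fill, outline and glow masks, with viewport-scaled edge smoothing.
# State parameters switch colors, and locking dims opacity and removes glow.
# A static switch enables a scrolling, wave-warped Honmoon texture; logo rotation and scale are parameterized.
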Material /BattlePassS37_Quests/Materials/M_UI_LevelTag
graph {
 properties {blendMode=BLEND_Translucent, materialDomain=MD_UI, numCustomizedUVs=1, pixelDepthOffsetMode=PDOM_Legacy, usedShadingModels=Default Lit}
 nodes [
  n1 = NamedRerouteUsage()
  n2 = NamedRerouteUsage()
  Multiply(
   A=LinearInterpolate(A=Max(A=@n1, B=@n2), B=ScalarParameter() {defaultValue=1, parameterName=FillOpacity}, Alpha=NamedRerouteUsage()) {constB=0.2}
   B=LinearInterpolate(B=ScalarParameter() {defaultValue=0.3, parameterName=LockedOpacity}, Alpha=ScalarParameter() {parameterName=isLocked}) {constA=1}
  )
  n3 = VectorParameter() {parameterName=GlowColor, defaultValue=<rgba 1 0 0.7231 1>}
  n4 = LinearInterpolate(
   A=LinearInterpolate(
    A=VectorParameter() {parameterName=IncompleteColor, defaultValue=<rgba 0.063 0.011 0.3185 1>}.RGB
    B=VectorParameter() {parameterName=CompleteColor, defaultValue=<rgba 1 0 0.7231 1>}.RGB
    Alpha=ScalarParameter() {parameterName=IsComplete}
   )
   B=VectorParameter() {parameterName=IncompleteColor, defaultValue=<rgba 0.063 0.011 0.3185 1>}.RGB
   Alpha=ScalarParameter() {parameterName=isLocked}
  )
  n5 = NamedRerouteUsage()
  Multiply(
   A=LinearInterpolate(
    A=LinearInterpolate(A=@n3.RGB, B=@n4, Alpha=@n5)
    B=VectorParameter() {parameterName=StrokeColor, defaultValue=<rgba 1 1 1 1>}.RGB
    Alpha=NamedRerouteUsage()
   )
   B=NamedRerouteUsage()
  )
  NamedRerouteDeclaration(None=AppendVector(
    A=ScalarParameter() {defaultValue=0.3, parameterName=HRhombusHeight, sortPriority=52}
    B=ScalarParameter() {defaultValue=0.7, parameterName=HRhombusWidth, sortPriority=53}
   )) {name=Rhombus2Dimensions, nodeColor=<rgba 0.6552 0.8698 0.0008 1>}
  NamedRerouteDeclaration(None=AppendVector(
    A=ScalarParameter() {defaultValue=0.3, parameterName=RhombusWidth, sortPriority=50}
    B=ScalarParameter() {defaultValue=0.8, parameterName=RhombusHeight, sortPriority=51}
   )) {name=RhombusDimensions, nodeColor=<rgba 0.6552 0.8698 0.0008 1>}
  n6 = MaterialFunctionCall(UVs=TextureCoordinate(), Texture Scale=ScalarParameter() {defaultValue=1.1, group=Honmoon, parameterName=HonMoonScale}) {materialFunction=&ScaleUVsByCenter}
  n7 = Constant() {r=0.06}
  n8 = Constant()
  n9 = Constant() {r=0.3}
  n10 = Constant() {r=0.01}
  n11 = Frac(None=Multiply(A=ScalarParameter() {defaultValue=0.05, group=Honmoon, parameterName=HonMoon_MotionSpeed}, B=Time()))
  n12 = SmoothStep(Value=TextureSample(UVs=MaterialFunctionCall!(UVs=@n6.UVs, Amplitude X=@n7, Amplitude Y=@n8, Frequency X=@n9, Frequency Y=@n10, Phase, Motion=@n11) {materialFunction=/FortUILibrary/MaterialFunctions/Effects/MF_UI_WaveWarp}.Warped UVs) {samplerType=SAMPLERTYPE_Alpha, texture=/BattlePassS37_Quests/Materials/Textures/T_UI_texture_Repeatable}.R) {constMax=2}
  n13 = Subtract(A=NamedRerouteUsage()) {constB=0.2}
  n14 = Multiply(A=NamedRerouteUsage(), B=OneMinus(None=SmoothStep(Value=@n13) {constMax=0.3, constMin=-0.5}))
  NamedRerouteDeclaration(None=StaticSwitchParameter(True=Multiply(A=@n12, B=@n14), False=Constant()) {group=Honmoon, parameterName=UseHonmoonTexture}) {name=HonMoonTexture, nodeColor=<rgba 0.0522 0.4285 1 1>}
  n15 = ScalarParameter() {defaultValue=0.5, parameterName=ChopOpacity}
  n16 = TextureCoordinate()
  n17 = ComponentMask(None=@n16) {r=true}
  NamedRerouteDeclaration(None=Max(A=OneMinus(None=@n15), B=SmoothStep(Value=@n17) {constMax=0.51, constMin=0.49})) {name=ChopMultiplyLayer, nodeColor=<rgba 0.0522 0.4285 1 1>}
  n18 = NamedRerouteUsage()
  n19 = NamedRerouteUsage()
  NamedRerouteDeclaration(None=OneMinus(None=SmoothStep(Max=@n18, Value=@n19) {constMax=0.01})) {name=OutlineMask, nodeColor=<rgba 0 0.1562 1 1>}
  n20 = NamedRerouteUsage()
  n21 = MaterialFunctionCall() {materialFunction=/FortUILibrary/MaterialFunctions/Basics/MF_UI_ViewportScale}
  NamedRerouteDeclaration(None=Add(A=Multiply(A=@n20, B=@n21.Viewport Scale Factor), B=Divide(A=@n20, B=@n21.Viewport Scale Factor))) {
   name=ShapeMaxStep
   desc="We want the SDF to scale (mult x viewport) but then we want to add back the quotient to bring back smoothing."
   nodeColor=<rgba 0 0.2734 1 1>
  }
  NamedRerouteDeclaration(None=ScalarParameter() {defaultValue=0.01, parameterName=ShapeSmoothing}) {name=ShapeSmoothing, nodeColor=<rgba 0 1 0.6016 1>}
  n22 = ScalarParameter() {parameterName=LogoRotation}
  NamedRerouteDeclaration(None=MaterialFunctionCall(
    UVs=MaterialFunctionCall(UVs=@n16, "Rotation Angle (0-1)"=@n22) {materialFunction=&CustomRotator}.Rotated Values
    Texture Scale=ScalarParameter() {defaultValue=1, parameterName=LogoScale}
   ) {materialFunction=&ScaleUVsByCenter}.UVs) {name=UVs, nodeColor=<rgba 0.0522 0.4285 1 1>}
  n23 = NamedRerouteUsage()
  n24 = NamedRerouteUsage()
  NamedRerouteDeclaration(None=OneMinus(None=SmoothStep(Max=@n23, Value=@n24))) {name=FillMask, nodeColor=<rgba 0 0.1562 1 1>}
  n25 = NamedRerouteUsage()
  n26 = MaterialFunctionCall!(DistanceFieldA=NamedRerouteUsage(), DistanceFieldB=NamedRerouteUsage(), BlendAmount) {materialFunction=/FortUILibrary/MaterialFunctions/Shapes/Modifiers/MF_UI_SDF_Union}
  NamedRerouteDeclaration(None=OneMinus(None=SmoothStep(Max=@n25, Value=@n26.Exact))) {name=LogoMask, nodeColor=<rgba 0 0.1562 1 1>}
  n27 = SmoothStep(
   Min=ScalarParameter() {defaultValue=-0.3, parameterName=GlowStrength_Min}
   Max=ScalarParameter() {defaultValue=0.3, parameterName=GlowStrength_Max}
   Value=NamedRerouteUsage()
  ) {constMax=0.35, constMin=-0.5}
  n28 = Add(A=Multiply(A=OneMinus(None=@n27), B=NamedRerouteUsage()), B=NamedRerouteUsage())
  n29 = ScalarParameter() {defaultValue=1, parameterName=GlowOpacity}
  NamedRerouteDeclaration(None=LinearInterpolate(A=Multiply(A=@n28, B=@n29), Alpha=ScalarParameter() {parameterName=isLocked}) {constA=1, constB=0}) {name=GlowMask, nodeColor=<rgba 0 0.1562 1 1>}
  n30 = MaterialFunctionCall!(UVs, Radius) {materialFunction=/FortUILibrary/MaterialFunctions/Shapes/SDFs/MF_SDF_Circle}
  NamedRerouteDeclaration(None=OneMinus(None=SmoothStep(Value=@n30.SDF) {constMax=0.5})) {name=Vignette, nodeColor=<rgba 0.0522 0.4285 1 1>}
  MaterialExpressionNamedRerouteUsage_4 = NamedRerouteUsage()
  MaterialExpressionNamedRerouteUsage_17 = NamedRerouteUsage()
  n31 = Custom!(p=@MaterialExpressionNamedRerouteUsage_4, b=@MaterialExpressionNamedRerouteUsage_17) {
   code="p = p - .5;\r\np = p*2;\r\np = abs(p);\r\n\r\nfloat nd = (b.x - 2.0*p.x)*b.x - (b.y - 2.0*p.y)*b.y;\r\nfloat bb = max(dot(b, b), 1e-8);\r\nfloat h = clamp(nd / bb, -1.0, 1.0);\r\nfloat2 q = 0.5 * b * float2(1.0 - h, 1.0 + h);\r\nfloat d = length(p - q);\r\nfloat s = p.x*b.y + p.y*b.x - b.x*b.y;\r\nreturn d * sign(s);"
   description=SDF_Rhombus
   inputs [
    {inputName=p, input {expression=&self:MaterialExpressionNamedRerouteUsage_4}}
    {inputName=b, input {expression=&self:MaterialExpressionNamedRerouteUsage_17}}
   ]
  }
  MaterialExpressionMaterialFunctionCall_14 = MaterialFunctionCall(
   UVs=@MaterialExpressionNamedRerouteUsage_4
   "Rotation Angle (0-1)"=ScalarParameter() {defaultValue=0.25, parameterName=RhombusIntersectRotation}
  ) {materialFunction=&CustomRotator}
  MaterialExpressionNamedRerouteUsage_12 = NamedRerouteUsage()
  n32 = Custom(p=@MaterialExpressionMaterialFunctionCall_14.Rotated Values, b=@MaterialExpressionNamedRerouteUsage_12) {
   code="p = p - .5;\r\np = p*2;\r\np = abs(p);\r\n\r\nfloat nd = (b.x - 2.0*p.x)*b.x - (b.y - 2.0*p.y)*b.y;\r\nfloat bb = max(dot(b, b), 1e-8);\r\nfloat h = clamp(nd / bb, -1.0, 1.0);\r\nfloat2 q = 0.5 * b * float2(1.0 - h, 1.0 + h);\r\nfloat d = length(p - q);\r\nfloat s = p.x*b.y + p.y*b.x - b.x*b.y;\r\nreturn d * sign(s);"
   description=SDF_Rhombus
   inputs [
    {inputName=p, input {expression=&self:MaterialExpressionMaterialFunctionCall_14}}
    {inputName=b, input {expression=&self:MaterialExpressionNamedRerouteUsage_12}}
   ]
  }
  n33 = MaterialFunctionCall(DistanceFieldA=@n31, DistanceFieldB=@n32, BlendAmount=ScalarParameter() {defaultValue=0.25, parameterName=LogoShapeBlend}) {materialFunction=&MF_UI_SDF_Union}
  NamedRerouteDeclaration(None=Subtract(A=Abs(None=@n33.Smooth), B=ScalarParameter() {defaultValue=0.02, parameterName=OutlineSize})) {name=OutlineSDF, nodeColor=<rgba 0 0.1562 1 1>}
  NamedRerouteDeclaration(None=@n33.Smooth) {name=FillSDF, nodeColor=<rgba 0 0.1562 1 1>}
 ]
}
